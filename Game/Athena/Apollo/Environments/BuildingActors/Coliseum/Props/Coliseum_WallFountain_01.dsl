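; Construction script for the Coliseum wall fountain prop Blueprint, a building actor in the Apollo environment.
; After running the parent construction script, it sets the relative location of the water fountain splash component from the SplashHeight variable.
; This lets the splash effect's position be adjusted per placed instance.
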(fn ConstructionScript
  (|CallParentFunction)
  (Transformation|SetRelativeLocation (Variables|Default|GetP_Colliseum_WaterFountain_Splash) (Variables|Default|GetSplashHeight) false false))
